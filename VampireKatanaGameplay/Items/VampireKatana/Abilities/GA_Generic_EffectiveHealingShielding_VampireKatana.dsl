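; Gameplay ability for the Vampire Katana that siphons damage dealt into healing or shields, gated by a scalable SiphonEnabled value.
; It waits for DamageDealt events with the required instigator tags and filters out non-pawn, same-team and downed targets.
; A rounded percentage of damage magnitude becomes the heal amount before Heal runs.
; Heal and shield cue overrides also replicate effects on the owner, since cues after the event wait are not predicted.
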(event Ability|EventActivateAbility
  ;[1 1 1] Activate Ability
  (if (>= (ScalableFloat|GetValueatLevel (Variables|Default|GetSiphonEnabled)) 1.0)
    (bind _outfortplayerpawn (CallFunction|GetAvatarasValidFortPlayerPawn)
      (:ValidFortPlayerPawn
        (Variables|Default|SetPawn _outfortplayerpawn)
        (bind (_asynctaskproxy _payload) (Ability|Tasks|WaitGameplayEvent (Variables|Default|GetEventTag_DamageDealt) 0 false false)
          (:EventReceived
            (if (GameplayTags|HasAllTags (.InstigatorTags (Utilities|Struct|BreakGameplayEventData _payload)) (Variables|Default|GetRequiredInstigatorTags))
              (bind _shouldheal (CallFunction|FilterforHeal (.Instigator (Utilities|Struct|BreakGameplayEventData _payload)) (.Target (Utilities|Struct|BreakGameplayEventData _payload))))
              (if _shouldheal
                (Variables|SetHeal_Amount_Value (Utilities|Struct|MakeScalableFloat (Math|Conversions|ToFloat(Integer) (Math|Float|Round (* (.EventMagnitude (Utilities|Struct|BreakGameplayEventData _payload)) (ScalableFloat|GetValueatLevel (Variables|Default|GetStealPercentage)))))))
                (CallFunction|Heal)))))))))

(event Custom|ReplicateFXOnOwner
  (params (GC_Tag :type Gameplay%20Tag%20Structure))
  ;[1 1 1] Replicate FX On Owner - We do this because ExcecuteGameplayCueOnOwner only works if the ability is being predicted actively and only the parts of ActivateAbility that are before any delays or waits are predictive. In this case, we are waiting on an event and need to replicate the local fx ourselves.
  (bind _outfortplayerpawn (CallFunction|GetAvatarasValidFortPlayerPawn)
    (:ValidFortPlayerPawn
      (bind _output_get (Variables|Default|SetPawn _outfortplayerpawn))
      (GameplayCue|ExecuteGameplayCue%28Burst%29Local _output_get GC_Tag (Ability|GameplayCue|MakeGameplayCueParameters 0.0)))))

(event Ability|EventOnEndAbility
  (params (bWasCancelled :type Boolean))
  (|CallParentFunction bWasCancelled) ; This node is disabled and will not be called.\nDrag off pins to build functionality.
)

(fn ExecuteGameplayCue_HealHealth
  (|CallParentFunction)
  (CallFunction|ReplicateFXOnOwner (Variables|GameplayCues|GetTag_GC_Heal)))

(fn ExecuteGameplayCue_HealShields
  (|CallParentFunction)
  (CallFunction|ReplicateFXOnOwner (Variables|GameplayCues|GetTag_GC_Shield)))

(fn FilterForHeal
  (params (Instigator :type Actor%20Object%20Reference) (ActorToFilter :type Actor%20Object%20Reference))
  (bind _asfort_pawn (Utilities|Casting|CastToFortPawn ActorToFilter)
    (:then
      (bind _output_get (LocalVariables|SetTargetAsFortPawn _asfort_pawn))
      (if (Team|OnSameTeam _output_get (Variables|GetInstigator))
        (return false)
        (elif (Health|IsDBNO (LocalVariables|GetTargetAsFortPawn))
          (return false)
          (else
            (return true)))))
    (:CastFailed
      (return false))))
